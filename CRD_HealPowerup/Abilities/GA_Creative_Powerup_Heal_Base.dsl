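; Gameplay ability base for the Creative heal powerup, activated from a gameplay event.
; It casts the event's optional object to the heal powerup and checks whether that powerup is instant.
; For non-instant heals, it first removes active effects tagged as duration-based heal powerups from the instigator's ability system.
; It then calls the parent activation with the same event data, after any earlier timed heal is cleared.
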(event Ability|EventActivateAbilityFromEvent
  (params (EventData :type Gameplay%20Event%20Data%20Structure%20%28by%20ref%29))
  (bind (_eventtag _instigator _target _optionalobject) (Utilities|Struct|BreakGameplayEventData EventData))
  (Utilities|FlowControl|Sequence
    (:then_0
      (bind _ascreative_powerup_heal (Utilities|Casting|CastToCreative_Powerup_Heal _optionalobject))
      (if (not (Class|CreativePowerupHeal|IsInstant _ascreative_powerup_heal))
        (GameplayEffects|RemoveActiveEffectswithTags (Ability|GetAbilitySystemComponent _instigator) "(GameplayTags=((TagName=\"Asset.Creative.CreativeEffect.HealPowerup.HasDuration\")))")))
    (:then_1
      (|CallParentFunction EventData))))
